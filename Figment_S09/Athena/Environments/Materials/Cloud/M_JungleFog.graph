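# Unlit additive jungle fog material for mesh-based fog, tinted by morning, day, evening and night colors through Color_Time_of_Day.
# Opacity combines tiled noise, a vertex-red HeightLerp mask, Fresnel-blended opacity values and depth fade.
# A camera-height factor, computed per vertex, fades the fog in above a height offset.
# When that factor reaches zero, a vertex offset collapses the mesh to its object position, hiding it entirely.
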Material /Figment_S09/Athena/Environments/Materials/Cloud/M_JungleFog
graph {
 properties {bUseTranslucencyVertexFog=false, blendMode=BLEND_Additive, pixelDepthOffsetMode=PDOM_Legacy, shadingModel=MSM_Unlit, usedShadingModels=Unlit}
 nodes [
  n1 = ComponentMask(None=CameraPositionWS()) {b=true}
  n2 = ScalarParameter() {defaultValue=-15000, parameterName=OffsetHeight}
  n3 = Clamp(None=Divide(A=Add(A=@n1, B=@n2), B=ScalarParameter() {defaultValue=20000, parameterName=HeightFade}))
  LinearInterpolate(A=Subtract(A=ObjectPositionWS(), B=WorldPosition().XYZ), B=Constant3Vector(), Alpha=Ceil(None=@n3))
  MaterialFunctionCall(
   Color_Morning=VectorParameter() {parameterName=JungleFogColorMorning, defaultValue=<rgba 0.4479 0.446 0.2403 1>}.RGB
   Color_Day=VectorParameter() {parameterName=JungleFogColorDay, defaultValue=<rgba 0.4479 0.446 0.2403 1>}.RGB
   Color_Evening=VectorParameter() {parameterName=JungleFogColorDayEvening, defaultValue=<rgba 0.4479 0.446 0.2403 1>}.RGB
   Color_Night=VectorParameter() {parameterName=JungleFogColorDayNight, defaultValue=<rgba 0.1796 0.224 0.1576 1>}.RGB
  ) {materialFunction=&Color_Time_of_Day}
  n4 = TextureSample(UVs=Multiply(A=TextureCoordinate(), B=ScalarParameter() {defaultValue=2.6129, parameterName=Tiling})) {texture=/Figment_S09/Effects/Fort_Effects/Textures/Misc/T_Tiling_Noise_01}
  n5 = ScalarParameter() {defaultValue=0.2381, parameterName=ContrastNoise}
  n6 = Multiply(
   A=Saturate(None=Add(A=@n4.R, B=@n5))
   B=MaterialFunctionCall(Transition Phase=VertexColor().R, Height Texture=@n4.R, Contrast=ScalarParameter() {defaultValue=0.0095, parameterName=Contrast}) {materialFunction=&HeightLerp}.Alpha
  )
  n7 = LinearInterpolate(
   A=ScalarParameter() {defaultValue=0.27, parameterName=Opacity}
   B=ScalarParameter() {parameterName=Opacity2}
   Alpha=Fresnel(ExponentIn=ScalarParameter() {defaultValue=3, parameterName=Fresnel}) {baseReflectFraction=0}
  )
  Multiply(
   A=DepthFade(Opacity=Multiply(A=@n6, B=@n7), FadeDistance=ScalarParameter() {defaultValue=4814.3574, parameterName=DepthFade})
   B=VertexInterpolator(VS=@n3).PS
  )
 ]
}
